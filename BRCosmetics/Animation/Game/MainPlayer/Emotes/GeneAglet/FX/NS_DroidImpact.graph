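# NS_DroidImpact is a Niagara impact effect for the GeneAglet emote, combining a flash, smoke puff and hot sparks.
# Flare drives sprite and light renderers, while Impact_Smoke bursts a few expanding, fading sprites tinted by User.SmokeColor.
# Two spark emitters burst speed-stretched sprites with gravity and drag.
# User.SparkSpawn sets flare and spark probability, those emitters scale and cull by distance, and front-end inputs tighten smoke radius and spark speed.
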NiagaraSystem /BRCosmetics/Animation/Game/MainPlayer/Emotes/GeneAglet/FX/NS_DroidImpact
dependencies {
 usedDataInterfaces [
  &NiagaraDataInterfaceCurve
  &NiagaraDataInterfaceMeshRendererInfo
  &NiagaraDataInterfaceVector2DCurve
  &NiagaraDataInterfaceColorCurve
  &NiagaraDataInterfaceVectorCurve
  &NiagaraDataInterfaceArrayFloat
 ]
 usedDynamicInputs [
  &Multiply_Float
  &Multiply_VectorByFloat
  &Vector2DFromFloat
 ]
 usedModules [&SystemState]
}
emitters {
 Flare {
  emitterUpdateScript [
   &EmitterState(
    Enable Distance Culling=true
    Max Distance Response=::KillImmediately
    Module.bApplySpawnCountScaleByDistance=true
    Module.bUseMaxDistance=true
    Module.bUseMinDistance=false
    Scalability Mode=::NewEnumerator1
    Scale Spawn Count=true
    Spawn Count Scale By Distance=DataInterface() {propertyValues {Curve {keys [<at 0.3676 0.9657>,<at 0.9962 0.4982>]}}}
   )
   &SpawnBurst_Instantaneous(Spawn Group=1, Spawn Probability=$User.SparkSpawn, Use Spawn Probability=true, UseLoopCountLimit=false) {module=SpawnBurst_Instantaneous001}
   &SpawnBurst_Instantaneous(Spawn Count=1, Spawn Probability=$User.SparkSpawn, Use Spawn Probability=true)
  ]
  particleSpawnScript [
   /SetVariables(Particles.VisibilityTag=&SpawnGroup()) {isSetParameters=true}
   &PartitionParticles()
   &InitializeParticle(
    Color=<rgba 1 0.5517 0.032 1>
    Lifetime=&SelectFloatFromArray(Direct Array Index=$Particles.PartitionParticles.Partition, Float Selection Array=DataInterface() {propertyValues {FloatData [0.2,0.08]}})
    Position Mode=::NewEnumerator0
    Sprite Rotation Angle=7.1362
    Sprite Rotation Mode=::NewEnumerator0
    Sprite Size Mode=::NewEnumerator3
    Uniform Sprite Size=120
    Write Material Random=true
   )
   &ShapeLocation(MaskSpawnGroup=false, Offset Mode=::NewEnumerator0, Override Seed=false, Sphere Radius=0, Sphere Surface Distribution=1, UseOwnerScale=true) {enabled=false}
   &AddVelocity(
    Module.UseConeFalloff=false
    Module.UseVelDistribution=false
    Override Seed=false
    Rotation Mode=::NewEnumerator2
    Velocity Mode=::NewEnumerator2
    Velocity Speed=1.3651
    Yaw / Pitch / Roll:Vector3f=<0 60 0>
   ) {enabled=false}
   &SpriteFacingAndAlignment(Sprite Alignment=<1 0 1.0829>, WriteAlignment=true, WriteFacing=true) {enabled=false}
  ]
  particleUpdateScript [
   &ParticleState()
   &SolveForcesAndVelocity(Clamp Velocity=false, Limit Acceleration=false)
   &ScaleColor(
    Linear Color Curve=DataInterface() {propertyValues {
      AlphaCurve {keys [<at 0 1>,<at 1 0>]}
      BlueCurve {keys [<at 0.2141 0.625>,<at 0.2356 0>,<at 1 0>]}
      GreenCurve {keys [<at 0.2141 2.7104>,<at 0.2356 0.9373>,<at 1 0.4174>]}
      RedCurve {keys [<at 0.2141 6>,<at 0.2356 1.7917>,<at 1 2>]}
     }}
    Scale Mode=::NewEnumerator2
   )
   &FN_Light_Attributes(
    Color=&ColorFromCurve(ColorCurve=DataInterface() {propertyValues {
       AlphaCurve {keys [<at 0 1 RCIM_Cubic>,<at 1 0 RCIM_Cubic>]}
       BlueCurve {keys [<at 0 0 RCIM_Cubic>,<at 0.9736 0>]}
       GreenCurve {keys [<at 0 1.2454 RCIM_Cubic>,<at 0.9736 0>]}
       RedCurve {keys [<at 0 2.79 RCIM_Cubic>,<at 0.9736 0>]}
      }})
    Create Color Attribute=true
    Create Exponent Attribute=true
    Create Radius Attribute=true
    Exponent=0.7
    Radius=120
    Scale Color=3
   )
   /SetVariables(
    Particles.HitImpactAlignment:Vector3f=<1 0 1>
    Particles.Vector 2D:Vector2f=/Niagara/DynamicInputs/ScaleByCurve/ScaleVector2DByCurve(
     InitialValue:Vector2f=<95 95>
     ScaleCurve:NiagaraDataInterfaceVectorCurve=DataInterface() {propertyValues {
       XCurve {keys [
         {arriveTangent=2.2184, interpMode=RCIM_Cubic, leaveTangent=2.2184, tangentMode=RCTM_User, time=0, value=0}
         {arriveTangent=-0.0316, tangentMode=RCTM_User, time=1, value=1}
        ]}
       YCurve {keys [
         {arriveTangent=2.2184, interpMode=RCIM_Cubic, leaveTangent=2.2184, tangentMode=RCTM_User, time=0, value=0}
         {arriveTangent=-0.0316, tangentMode=RCTM_User, time=1, value=1}
        ]}
       ZCurve {keys []}
      }}
     CurveIndex:NiagaraFloat=$Particles.NormalizedAge
    )
    Particles.hitimpactcolor:LinearColor=&ColorFromCurve(ColorCurve=DataInterface() {propertyValues {
       AlphaCurve {keys [<at 0.333 1>]}
       BlueCurve {keys [<at 0 0.24>,<at 0.166 0.2>,<at 0.7498 0>,<at 0.9196 0>]}
       GreenCurve {keys [<at 0 3.12>,<at 0.166 2.6>,<at 0.7498 0.6667>,<at 0.9196 0.15>]}
       RedCurve {keys [<at 0 6>,<at 0.166 5>,<at 0.7498 4>,<at 0.9196 3>]}
      }})
   ) {isSetParameters=true}
  ]
  rendererClasses [&NiagaraSpriteRendererProperties, &NiagaraLightRendererProperties]
  renderers [
   {rendererClass=&NiagaraSpriteRendererProperties}
   {rendererClass=&NiagaraLightRendererProperties, rendererIndex=1}
   {rendererClass=&NiagaraSpriteRendererProperties, rendererIndex=2}
  ]
 }
 Impact_Smoke {
  emitterUpdateScript [
   &EmitterState(Inactive Response=::NewEnumerator1, Life Cycle Mode=::NewEnumerator1, Loop Behavior=::NewEnumerator1, Loop Duration=0.5, UseLoopDelay=true)
   &SpawnBurst_Instantaneous(
    Spawn Count=&RandomRangeInt(Minimum=2, Maximum=3, Recalculate Random Each Loop=false)
    Use Spawn Probability=false
    UseLoopCountLimit=false
   )
  ]
  particleSpawnScript [
   &InitializeParticle(
    Color=$User.SmokeColor
    Lifetime Max=0.5
    Lifetime Min=0.3
    Lifetime Mode=::NewEnumerator1
    Sprite Rotation Mode=::NewEnumerator2
    Sprite Size Mode=::NewEnumerator1
    Uniform Sprite Size Max=80
    Uniform Sprite Size Min=50
    UsePositionOffset=false
    Write Material Random=true
   )
   &CameraOffset(Camera Offset Amount=&RandomRangeFloat(Minimum=40, Maximum=50, Evaluation Type=::NewEnumerator1)) {enabled=false}
   &ShapeLocation(
    MaskSpawnGroup=false
    Override Seed=false
    Sphere Radius=&MakeCustomFloatFromBool(Bool=$User.bIsFrontEnd, True Float=5, False Float=10)
    UseOwnerScale=true
   )
   &AddVelocity(Module.Use Falloff Distance=false, Override Seed=false, Velocity Mode=::NewEnumerator1, Velocity Speed=12.4296)
  ]
  particleUpdateScript [
   &ParticleState()
   &ScaleSpriteSize(Uniform Curve Sprite Scale=DataInterface() {propertyValues {Curve {keys [
        {arriveTangent=1.1861, interpMode=RCIM_Cubic, leaveTangent=1.1861, tangentMode=RCTM_User, time=0, value=0.35}
        {interpMode=RCIM_Cubic, tangentMode=RCTM_User, time=1, value=1}
       ]}}})
   &ScaleColor(
    Scale Alpha=&FloatFromCurve(FloatCurve=DataInterface() {propertyValues {Curve {keys [
         {arriveTangent=2.4358, interpMode=RCIM_Cubic, leaveTangent=2.4358, tangentMode=RCTM_User, time=0, value=0}
         <at 0.25 0.5 RCIM_Cubic>
         <at 1 0 RCIM_Cubic>
        ]}}})
    Scale RGB=&VectorFromCurve(VectorCurve=DataInterface() {propertyValues {XCurve {keys [<at 0 1>,<at 1 1>]}, YCurve {keys [<at 0 1>,<at 1 1>]}, ZCurve {keys [<at 0 1>,<at 1 1>]}}})
    ScaleA=true
    ScaleRGB=true
   )
   &DynamicMaterialParameters(
    Index 0 Param 1=&Add_Float(
     A=&RandomRangeFloat(Minimum=0, Maximum=1)
     B=&FloatFromCurve(
      FloatCurve=DataInterface() {propertyValues {Curve {keys [
          <at 0 0 RCIM_Cubic>
          {arriveTangent=1.5947, interpMode=RCIM_Cubic, leaveTangent=1.5947, tangentMode=RCTM_User, time=0.2557, value=0.5257}
          <at 1 1 RCIM_Cubic>
         ]}}}
      CurveIndex=$Particles.Age
      Scale Curve=0.25
     )
    )
    Module.Param0WriteEnabled=true
    Module.Param1WriteEnabled=true
    Module.Param2WriteEnabled=true
    Module.Param3WriteEnabled=true
   ) {module=DynamicMaterialParameters001}
   &SolveForcesAndVelocity(Clamp Velocity=false, Limit Acceleration=false)
  ]
  rendererClasses [&NiagaraSpriteRendererProperties]
  renderers [
   {rendererClass=&NiagaraSpriteRendererProperties}
   {rendererClass=&NiagaraSpriteRendererProperties, rendererIndex=1}
  ]
 }
 Sparks_CPU {
  emitterUpdateScript [
   &EmitterState(
    Enable Distance Culling=true
    Life Cycle Mode=::NewEnumerator1
    Loop Behavior=::NewEnumerator1
    Loop Duration=0.3
    Max Distance Response=::KillImmediately
    Module.bApplySpawnCountScaleByDistance=true
    Module.bUseMaxDistance=true
    Module.bUseMinDistance=false
    Scalability Mode=::NewEnumerator1
    Scale Spawn Count=true
    Spawn Count Scale By Distance=DataInterface() {propertyValues {Curve {keys [<at 0.3676 0.9657>,<at 0.9962 0.4982>]}}}
    UseLoopDelay=false
   )
   &SpawnBurst_Instantaneous(
    Spawn Count=&UniformRangedInt(Minimum=25, Maximum=35)
    Spawn Probability=$User.SparkSpawn
    Use Spawn Probability=true
   )
  ]
  particleSpawnScript [
   &InitializeParticle(
    Color=&RandomRangeLinearColor(Minimum=<rgba 1 0.2103 0.2083 1>, Maximum=<rgba 1 1 1 1>)
    Lifetime Max=0.6
    Lifetime Min=0.3
    Lifetime Mode=::NewEnumerator1
    Sprite Size Max=<1.3 6>
    Sprite Size Min=<1.3536 0.5>
    Sprite Size Mode=::NewEnumerator2
    UsePositionOffset=false
    Write Material Random=true
   )
   &ShapeLocation(MaskSpawnGroup=false, Override Seed=false, Sphere Radius=10, Sphere Surface Distribution=1, UseOwnerScale=true)
   &AddVelocity(
    Module.Use Falloff Distance=false
    Override Seed=false
    Velocity Mode=::NewEnumerator1
    Velocity Speed=&FN_FrontEndFloatModifier(
     Gameplay=&RandomRangeFloat(Minimum=222, Maximum=666)
     Front End=&RandomRangeFloat(Minimum=150, Maximum=400)
    )
   )
   &DynamicMaterialParameters(
    Index 0 Param 1=3
    Index 0 Param 2=20
    Index 0 Param 3=1
    Module.Param0WriteEnabled=true
    Module.Param1WriteEnabled=true
    Module.Param2WriteEnabled=true
    Module.Param3WriteEnabled=true
   )
   &AddVelocity(Module.UseConeFalloff=false, Module.UseVelDistribution=false, Override Seed=false, Velocity Mode=::NewEnumerator2, Velocity Speed=250) {module=AddVelocity001}
  ]
  particleUpdateScript [
   &ParticleState()
   &GravityForce(Gravity=<0 0 -666>)
   &Drag(Drag=&UniformRangedFloat(Minimum=2, Maximum=6), Use Linear Drag=true, Use Rotational Drag=false)
   &SolveForcesAndVelocity(Clamp Velocity=false, Limit Acceleration=false)
   &ScaleSpriteSize(Scale Factor=&Vector2DFromCurve(
     Vector2Curve=DataInterface() {propertyValues {
       XCurve {keys [
         {arriveTangent=5.9728, interpMode=RCIM_Cubic, leaveTangent=5.9728, tangentMode=RCTM_User, time=0, value=0.1}
         {arriveTangent=-0.1, interpMode=RCIM_Cubic, leaveTangent=-0.1, tangentMode=RCTM_User, time=0.18, value=0.8}
         {interpMode=RCIM_Cubic, tangentMode=RCTM_User, time=1, value=0}
        ]}
       YCurve {keys [
         {arriveTangent=6.2715, interpMode=RCIM_Cubic, leaveTangent=6.2715, tangentMode=RCTM_User, time=0, value=0.1}
         {arriveTangent=-0.1001, interpMode=RCIM_Cubic, leaveTangent=-0.1001, tangentMode=RCTM_User, time=0.18, value=0.8}
         {interpMode=RCIM_Cubic, tangentMode=RCTM_User, time=0.9988, value=0}
        ]}
      }}
     ScaleCurve=<1 1>
    ))
   &ScaleSpriteSizeBySpeed(Max Scale Factor=<2 11.8122>, Min Scale Factor=<1.5 1>, Velocity Threshold=666)
   &ScaleColor(
    Linear Color Curve=DataInterface() {propertyValues {
      AlphaCurve {keys [<at 0.333 1>,<at 1 0>]}
      BlueCurve {keys [<at 0 0.24>,<at 0.166 0.2>,<at 0.333 0>,<at 0.75 0>]}
      GreenCurve {keys [<at 0 3.12>,<at 0.166 2.6>,<at 0.333 0.6667>,<at 0.75 0.15>]}
      RedCurve {keys [<at 0 6>,<at 0.166 5>,<at 0.333 4>,<at 0.75 3>]}
     }}
    Scale Mode=::NewEnumerator2
   )
  ]
  rendererClasses [&NiagaraSpriteRendererProperties]
  renderers [{rendererClass=&NiagaraSpriteRendererProperties}]
 }
 Sparks_CPU001 {
  emitterUpdateScript [
   &EmitterState(
    Enable Distance Culling=true
    Max Distance Response=::KillImmediately
    Module.bApplySpawnCountScaleByDistance=true
    Module.bUseMaxDistance=true
    Module.bUseMinDistance=false
    Scalability Mode=::NewEnumerator1
    Scale Spawn Count=true
    Spawn Count Scale By Distance=DataInterface() {propertyValues {Curve {keys [<at 0.3676 0.9657>,<at 0.9962 0.4982>]}}}
   )
   &SpawnBurst_Instantaneous(
    Spawn Count=&UniformRangedInt(Minimum=10, Maximum=15)
    Spawn Probability=$User.SparkSpawn
    Use Spawn Probability=true
   )
  ]
  particleSpawnScript [
   &InitializeParticle(
    Color=&RandomRangeLinearColor(Minimum=<rgba 1 0.2103 0.2083 1>, Maximum=<rgba 1 1 1 1>)
    Lifetime Max=1
    Lifetime Min=0.6
    Lifetime Mode=::NewEnumerator1
    Sprite Size Max=<1.3 6>
    Sprite Size Min=<1.3536 0.5>
    Sprite Size Mode=::NewEnumerator2
    UsePositionOffset=false
    Write Material Random=true
   )
   &ShapeLocation(MaskSpawnGroup=false, Override Seed=false, Sphere Radius=10, Sphere Surface Distribution=1, UseOwnerScale=true) {enabled=false}
   &AddVelocity(
    Module.Use Falloff Distance=false
    Override Seed=false
    Velocity Mode=::NewEnumerator1
    Velocity Speed=&FN_FrontEndFloatModifier(
     Gameplay=&RandomRangeFloat(Minimum=222, Maximum=666)
     Front End=&RandomRangeFloat(Minimum=150, Maximum=400)
    )
   )
   &DynamicMaterialParameters(
    Index 0 Param 1=3
    Index 0 Param 2=20
    Index 0 Param 3=1
    Module.Param0WriteEnabled=true
    Module.Param1WriteEnabled=true
    Module.Param2WriteEnabled=true
    Module.Param3WriteEnabled=true
   )
   &AddVelocity(Module.UseConeFalloff=false, Module.UseVelDistribution=false, Override Seed=false, Velocity Mode=::NewEnumerator2, Velocity Speed=250) {module=AddVelocity001}
  ]
  particleUpdateScript [
   &ParticleState()
   &GravityForce()
   &Drag(Drag=&UniformRangedFloat(Minimum=3, Maximum=8), Use Linear Drag=true, Use Rotational Drag=false)
   &SolveForcesAndVelocity(Clamp Velocity=false, Limit Acceleration=false)
   &ScaleSpriteSize(Scale Factor=&Vector2DFromCurve(
     Vector2Curve=DataInterface() {propertyValues {
       XCurve {keys [
         {arriveTangent=5.9728, interpMode=RCIM_Cubic, leaveTangent=5.9728, tangentMode=RCTM_User, time=0, value=0.1}
         {arriveTangent=-0.1, interpMode=RCIM_Cubic, leaveTangent=-0.1, tangentMode=RCTM_User, time=0.18, value=0.8}
         {interpMode=RCIM_Cubic, tangentMode=RCTM_User, time=1, value=0}
        ]}
       YCurve {keys [
         {arriveTangent=6.2715, interpMode=RCIM_Cubic, leaveTangent=6.2715, tangentMode=RCTM_User, time=0, value=0.1}
         {arriveTangent=-0.1001, interpMode=RCIM_Cubic, leaveTangent=-0.1001, tangentMode=RCTM_User, time=0.18, value=0.8}
         {interpMode=RCIM_Cubic, tangentMode=RCTM_User, time=0.9988, value=0}
        ]}
      }}
     ScaleCurve=<1 1>
    ))
   &ScaleSpriteSizeBySpeed(Max Scale Factor=<2 5>, Min Scale Factor=<1.5 1>, Velocity Threshold=666)
   &ScaleColor(
    Linear Color Curve=DataInterface() {propertyValues {
      AlphaCurve {keys [<at 0.333 1>,<at 1 0>]}
      BlueCurve {keys [<at 0 0.24>,<at 0.166 0.2>,<at 0.333 0>,<at 0.75 0>]}
      GreenCurve {keys [<at 0 3.12>,<at 0.166 2.6>,<at 0.333 0.6667>,<at 0.75 0.15>]}
      RedCurve {keys [<at 0 6>,<at 0.166 5>,<at 0.333 4>,<at 0.75 3>]}
     }}
    Scale Mode=::NewEnumerator2
   )
  ]
  rendererClasses [&NiagaraSpriteRendererProperties]
  renderers [{rendererClass=&NiagaraSpriteRendererProperties}]
 }
}
summary {
 systemName=NS_DroidImpact
 emitters [
  {emitterName=Sparks_CPU, rendererClasses [&NiagaraSpriteRendererProperties]}
  {emitterName=Impact_Smoke, rendererClasses [&NiagaraSpriteRendererProperties]}
  {emitterName=Flare, rendererClasses [&NiagaraSpriteRendererProperties, &NiagaraLightRendererProperties]}
  {emitterName=Sparks_CPU001, rendererClasses [&NiagaraSpriteRendererProperties]}
 ]
 userVariables [
  User.bIsFrontEnd {type=NiagaraBool}
  User.bIsFrontEndPreview {type=NiagaraBool}
  User.SmokeColor {type=LinearColor, description="", defaultValue {struct=LinearColor, value=<rgba 1 1 1 1>}}
  User.SparkSpawn {type=NiagaraFloat, defaultValue=1, description=""}
 ]
}
