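# Niagara system for the GeneAglet player emote's laser shots, with beam, impact, sparks and smoke.
# Beam_Impact bursts short-lived sprites and lights at two forward offsets split by partition, scaled by User Laser Distance.
# Sparks burst from the far end, falling and stretching with speed.
# SmokeRibbon trails from a skeletal mesh, with User NormalizedNotifyProgress driving spawn cutoff, curl noise and push, and SmokeHeight scaling rate, lifetime and velocity.
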NiagaraSystem /BRCosmetics/Animation/Game/MainPlayer/Emotes/GeneAglet/FX/NS_GeneAglet_Lasers
dependencies {
 usedDataInterfaces [
  &NiagaraDataInterfaceArrayFloat3
  &NiagaraDataInterfaceCurve
  &NiagaraDataInterfaceVector2DCurve
  &NiagaraDataInterfaceVectorCurve
  &NiagaraDataInterfaceColorCurve
 ]
 usedDynamicInputs [&FNTimeOfDayScalar, &ReturnNormalizedExecIndex]
 usedModules [&SystemState]
}
emitters {
 Beam_Impact {
  emitterUpdateScript [
   &EmitterState(Life Cycle Mode=::NewEnumerator1, Loop Behavior=::NewEnumerator1, Loop Duration=0.2, UseLoopDelay=false)
   &SpawnBurst_Instantaneous(Spawn Time=0.01, Use Spawn Probability=false, UseLoopCountLimit=false)
   &SpawnBurst_Instantaneous(Spawn Group=1, Spawn Time=0.01, Use Spawn Probability=false, UseLoopCountLimit=false) {module=SpawnBurst_Instantaneous002}
  ]
  particleSpawnScript [
   &PartitionParticles()
   &InitializeParticle(
    Color=<rgba 2.79 0.9623 0.0581 1>
    Lifetime=0.1
    Position Offset=&Multiply_VectorByFloat(
     Vector=&SelectVectorFromArray(
      Direct Array Index=$Particles.PartitionParticles.Partition
      Vector Selection Array=DataInterface() {propertyValues {FloatData [<8.221 0 0>, <40 0 0>]}}
     )
     Float=$User.Laser Distance
    )
    Position Offset Coordinate Space=::Local
    Sprite Size=&Multiply_Vector2D_ByFloat(Float=&Multiply_Float(A=$User.Laser Distance, B=1.2), Vector2D=<0.45 26.1663>)
    Sprite Size Mode=::NewEnumerator4
    UsePositionOffset=true
    Write Material Random=true
   )
   &ApplyInitialForces()
   &AddVelocity(Velocity=&Multiply_VectorByFloat(Vector=<1 0 0>, Float=$User.Laser Distance))
   /SetVariables(Particles.VisibilityTag=&SpawnGroup()) {isSetParameters=true}
  ]
  particleUpdateScript [
   &ParticleState()
   &ScaleSpriteSize(
    Non-Uniform Curve Sprite Scale=DataInterface() {propertyValues {
      XCurve {keys [
        <at 0 0 RCIM_Cubic>
        {arriveTangent=1.9773, interpMode=RCIM_Cubic, leaveTangent=1.9773, tangentMode=RCTM_User, time=0.2, value=0.7}
        {arriveTangent=-1.9026, interpMode=RCIM_Cubic, leaveTangent=-1.9026, tangentMode=RCTM_User, time=0.8, value=0.7}
        {tangentMode=RCTM_User, time=1, value=0}
       ]}
      YCurve {keys [{tangentMode=RCTM_User, time=0, value=1}, {tangentMode=RCTM_User, time=1, value=1}]}
     }}
    Scale Sprite Size Mode=::NewEnumerator3
   )
   &ScaleColor(
    Scale Alpha=&FloatFromCurve(FloatCurve=DataInterface() {propertyValues {Curve {keys [<at -0.0046 1.0113 RCIM_Cubic>,<at 1.0011 0.002 RCIM_Cubic>]}}})
    ScaleA=true
    ScaleRGB=false
   ) {enabled=false}
   &InheritSourceMovement(ApplyForce=false, ApplyPosition=true, ApplyPositionFalloff=false)
   &SolveForcesAndVelocity(Clamp Velocity=false, Limit Acceleration=false)
   /SetVariables(
    Particles.Linear Color=&ColorFromCurve(ColorCurve=DataInterface() {propertyValues {
       AlphaCurve {keys [<at 0 1 RCIM_Cubic>,<at 1 0 RCIM_Cubic>]}
       BlueCurve {keys [<at 0 0 RCIM_Cubic>]}
       GreenCurve {keys [<at 0 0.6263 RCIM_Cubic>]}
       RedCurve {keys [<at 0 5 RCIM_Cubic>]}
      }})
    Particles.impactsize:Vector2f=/Niagara/DynamicInputs/ScaleByCurve/ScaleVector2DByCurve(
     InitialValue:Vector2f=<17 17>
     ScaleCurve:NiagaraDataInterfaceVectorCurve=DataInterface() {propertyValues {XCurve {keys [<at 0 1>,<at 1 0>]}, YCurve {keys [<at 0 1>,<at 1 0>]}, ZCurve {keys []}}}
     CurveIndex:NiagaraFloat=$Particles.NormalizedAge
    )
    Particles.impactsubuv:NiagaraFloat=&RandomRangeFloat(Minimum=0, Maximum=3)
   ) {isSetParameters=true}
   &FN_Light_Attributes(
    Color=&ColorFromCurve(ColorCurve=DataInterface() {propertyValues {
       AlphaCurve {keys [<at 0 1 RCIM_Cubic>,<at 1 0 RCIM_Cubic>]}
       BlueCurve {keys [<at 0 0 RCIM_Cubic>,<at 1 0 RCIM_Cubic>]}
       GreenCurve {keys [<at 0 0.9972 RCIM_Cubic>,<at 1 0 RCIM_Cubic>]}
       RedCurve {keys [<at 0 2.79 RCIM_Cubic>,<at 1 0 RCIM_Cubic>]}
      }})
    Create Color Attribute=true
    Create Radius Attribute=true
    Radius=50
   )
   /SetVariables(
    Particles.DynamicParameterGLow:Vector4f=<1 0.1914 0 -0.5081>
    Particles.GlowSize=&Multiply_Vector2D_ByFloat(
     Float=&Multiply_Float(A=$User.Laser Distance, B=1.2)
     Vector2D=&ScaleVector2DByCurve(
      InitialValue:Vector2f=<10 26.486>
      ScaleCurve:NiagaraDataInterfaceVectorCurve=DataInterface() {propertyValues {
        XCurve {keys [
          {interpMode=RCIM_Cubic, tangentMode=RCTM_User, time=0, value=0}
          {
           arriveTangent=1.1108
           arriveTangentWeight=0.0667
           interpMode=RCIM_Cubic
           leaveTangent=1.1108
           leaveTangentWeight=0.2667
           tangentMode=RCTM_User
           time=0.2
           value=0.7
          }
          {
           arriveTangent=-3.1987
           arriveTangentWeight=0.2
           interpMode=RCIM_Cubic
           leaveTangent=-3.1987
           leaveTangentWeight=0.0667
           tangentMode=RCTM_User
           time=0.8
           value=0.35
          }
          {tangentMode=RCTM_User, time=1, value=0}
         ]}
        YCurve {keys [<at 0 1>,<at 1 1>]}
        ZCurve {keys []}
       }}
      CurveIndex:NiagaraFloat=$Particles.NormalizedAge
     )
    )
    Particles.beamcolor:LinearColor=<rgba 1 0.658 0.4212 1>
    Particles.glowcolor:LinearColor=<rgba 5 0.9121 0 0.2348>
   ) {isSetParameters=true}
  ]
  rendererClasses [&NiagaraSpriteRendererProperties, &NiagaraLightRendererProperties]
  renderers [
   {rendererClass=&NiagaraSpriteRendererProperties}
   {rendererClass=&NiagaraSpriteRendererProperties, rendererIndex=1}
   {rendererClass=&NiagaraLightRendererProperties, rendererIndex=2}
   {rendererClass=&NiagaraSpriteRendererProperties, rendererIndex=3}
  ]
 }
 MuzzleFlash {rendererClasses [], renderers []}
 SmokeRibbon {
  emitterUpdateScript [
   &EmitterState(Life Cycle Mode=::NewEnumerator1, Loop Behavior=::NewEnumerator1, Loop Duration=1, UseLoopDelay=false)
   &SpawnRate(
    SpawnRate=&Multiply_Float(
     A=&Multiply_Float(A=70, B=$User.SmokeHeight)
     B=&FloatFromCurve(FloatCurve=DataInterface() {propertyValues {Curve {keys [<at 0.5 1>,<at 0.6 0>]}}}, CurveIndex=$User.NormalizedNotifyProgress)
    )
    Use Spawn Probability=false
   )
  ]
  particleSpawnScript [
   &InitializeParticle(
    Color=<rgba 1 1 1 0.05>
    Lifetime=&Multiply_Float(A=0.3, B=$User.SmokeHeight)
    Position Offset Coordinate Space=::Local
    Ribbon Width Mode=::NewEnumerator1
    UsePositionOffset=true
    Write Material Random=true
   )
   &AddVelocity(Rotation Coordinate Space=::World, Velocity=<-3 0 0>, Velocity Speed Scale=$User.SmokeHeight)
   &InheritVelocity(Inherited Velocity Amount Scale=&VectorFromFloat(Value=0.2), Speed Limit=true, UseSpeedThreshold=false) {enabled=false}
   &SkeletalMeshLocation(MaskSpawnGroup=false, Mesh Sampling Type=::NewEnumerator5, Skeletal Mesh=$User.Skeletal Mesh Ribbon, Socket Sampling Mode=::NewEnumerator5)
  ]
  particleUpdateScript [
   &ParticleState()
   /Niagara/Modules/Update/Forces/CurlNoiseForce(
    Noise Frequency:NiagaraFloat=&Multiply_Float(A=0.0001, B=1)
    Noise Strength=&FloatFromCurve(
     FloatCurve=DataInterface() {propertyValues {Curve {keys [{arriveTangent=592.1932, interpMode=RCIM_Cubic, leaveTangent=592.1932, tangentMode=RCTM_User, time=0, value=100}, <at 1 300>]}}}
     Scale Curve=&Add_Float(
      A=&Multiply_Float(A=0.7, B=$User.SmokeHeight)
      B=&FloatFromCurve(
       FloatCurve=DataInterface() {propertyValues {Curve {keys [{interpMode=RCIM_Cubic, tangentMode=RCTM_User, time=0.45, value=0}, {interpMode=RCIM_Cubic, tangentMode=RCTM_User, time=1, value=1}]}}}
       CurveIndex=$User.NormalizedNotifyProgress
       Scale Curve=10
      )
     )
    )
    Pan Noise=true
    Pan Noise Field=<0 0 2>
    Random Seed:NiagaraInt32=&RandomRangeInt(Minimum=0, Maximum=100)
   ) {enabled=false}
   /Niagara/Modules/Update/Forces/V2/CurlNoiseForce(
    Noise Frequency=119.6126
    Noise Strength=&FloatFromCurve(
     FloatCurve=DataInterface() {propertyValues {Curve {keys [{interpMode=RCIM_Cubic, tangentMode=RCTM_User, time=0, value=0.1}, {interpMode=RCIM_Cubic, tangentMode=RCTM_User, time=1, value=1}]}}}
     CurveIndex=$User.NormalizedNotifyProgress
     Scale Curve=800
    )
    Pan Noise=true
    Pan Noise Field=<0 0 1>
   ) {module=CurlNoiseForce001}
   /SetVariables(Transient.PhysicsForce:Vector3f=&Add_Vector(
     A=&Multiply_Vector(A=$Output.CurlNoiseForce001.CurlNoiseForce, B=<1 1 0>)
     B=&Multiply_VectorByFloat(
      Vector=&TransformVector(Vector=<-1 0 0>, Destination Space=::World)
      Float=&FloatFromCurve(
       FloatCurve=DataInterface() {propertyValues {Curve {keys [
           {arriveTangent=5.1592, interpMode=RCIM_Cubic, leaveTangent=5.1592, tangentMode=RCTM_User, time=0.45, value=0}
           {interpMode=RCIM_Cubic, tangentMode=RCTM_User, time=1, value=1}
          ]}}}
       CurveIndex=$User.NormalizedNotifyProgress
       Scale Curve=800
      )
     )
    )) {isSetParameters=true}
   &ScaleRibbonWidth(Ribbon Width Scale=&FloatFromCurve(FloatCurve=DataInterface() {propertyValues {Curve {keys [
         {arriveTangent=0.1406, interpMode=RCIM_Cubic, leaveTangent=0.1406, tangentMode=RCTM_User, time=0, value=1}
         {arriveTangent=-0.2815, tangentMode=RCTM_User, time=1, value=0}
        ]}}})) {module=ScaleRibbonWidth001}
   &ScaleColor(Scale Mode=::NewEnumerator2)
   &SolveForcesAndVelocity(Clamp Velocity=false, Limit Acceleration=false)
  ]
  rendererClasses [&NiagaraRibbonRendererProperties]
  renderers [{rendererClass=&NiagaraRibbonRendererProperties}]
 }
 Sparks {
  emitterUpdateScript [
   &EmitterState(Life Cycle Mode=::NewEnumerator1, Loop Behavior=::NewEnumerator1, Loop Duration=1, UseLoopDelay=false)
   &SpawnBurst_Instantaneous(Spawn Count=12, Spawn Time=0.01, Use Spawn Probability=false, UseLoopCountLimit=false)
  ]
  particleSpawnScript [
   &InitializeParticle(
    Color=<rgba 10 0.7126 0 1>
    Lifetime Max=0.15
    Lifetime Min=0.08
    Lifetime Mode=::NewEnumerator1
    Sprite Size Mode=::NewEnumerator1
    Uniform Sprite Size Max=0.8
    Uniform Sprite Size Min=0.4
    UsePositionOffset=false
    Write Material Random=true
   )
   &ShapeLocation(
    MaskSpawnGroup=false
    Offset=&Multiply_VectorByFloat(Vector=<40 0 0>, Float=$User.Laser Distance)
    Offset Coordinate Space=::Local
    Offset Mode=::NewEnumerator0
    Override Seed=false
    Sphere Radius=0
    UseOwnerScale=true
   )
   &AddVelocity(
    Cone Angle=120
    Module.UseConeFalloff=false
    Module.UseVelDistribution=false
    Override Seed=false
    Velocity Mode=::NewEnumerator2
    Velocity Speed=&RandomRangeFloat(Minimum=300, Maximum=800)
   )
  ]
  particleUpdateScript [
   &ParticleState()
   &ScaleSpriteSize(Uniform Curve Sprite Scale=DataInterface() {propertyValues {Curve {defaultValue=0, keys [{tangentMode=RCTM_User, time=0, value=1}, {tangentMode=RCTM_User, time=1, value=0}]}}}) {module=ScaleSpriteSize001}
   &Drag(Drag=&RandomRangeFloat(Minimum=0.8, Maximum=1.2), Use Linear Drag=true, Use Rotational Drag=false)
   &GravityForce(Gravity=<0 0 -5518.2236>)
   &SolveForcesAndVelocity(Clamp Velocity=false, Limit Acceleration=false)
   &ScaleSpriteSizeBySpeed(Max Scale Factor=<2 9.6922>, Min Scale Factor=<1 17.4583>)
   &ScaleColor(Scale Mode=::NewEnumerator2)
  ]
  rendererClasses [&NiagaraSpriteRendererProperties]
  renderers [{rendererClass=&NiagaraSpriteRendererProperties}]
 }
}
summary {
 systemName=NS_GeneAglet_Lasers
 emitters [
  {emitterName=Beam_Impact, rendererClasses [&NiagaraSpriteRendererProperties, &NiagaraLightRendererProperties]}
  {emitterName=Sparks, rendererClasses [&NiagaraSpriteRendererProperties]}
  {emitterName=SmokeRibbon, rendererClasses [&NiagaraRibbonRendererProperties]}
  {emitterName=MuzzleFlash, rendererClasses []}
 ]
 userVariables [
  User.Laser Distance {type=NiagaraFloat, defaultValue=1, description=""}
  User.NormalizedNotifyProgress {type=NiagaraFloat, defaultValue=0.4, description=""}
  User.Skeletal Mesh Ribbon {
   type=NiagaraDataInterfaceSkeletalMesh
   defaultValue {
    struct=NiagaraExt_VariableValue_DataInterface
    value {dataInterface=&self:NiagaraDataInterfaceSkeletalMesh_0}
   }
  }
  User.SmokeHeight {type=NiagaraFloat, defaultValue=1, description=""}
 ]
}
