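# Niagara system for the prop-spawn effect of the GeneAglet main-player emote.
# It combines three emitters, Airglow, Smoke and Sparkles, whose LWE suffix suggests lightweight emitters; their internals and renderers are not listed.
# SystemState is the only listed module.
# User parameters expose two skeletal mesh interfaces, normalized notify progress, a ribbon ID offset and a generic vector, with no visible links to the emitters.
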NiagaraSystem /BRCosmetics/Animation/Game/MainPlayer/Emotes/GeneAglet/FX/NS_GeneAglet_PropSpawn
dependencies {usedDataInterfaces [], usedDynamicInputs [], usedModules [&SystemState]}
emitters {AirglowLWE {rendererClasses [], renderers []}, SmokeLWE {rendererClasses [], renderers []}, SparklesLWE {rendererClasses [], renderers []}}
summary {
 systemName=NS_GeneAglet_PropSpawn
 emitters [{emitterName=SmokeLWE, rendererClasses []}, {emitterName=AirglowLWE, rendererClasses []}, {emitterName=SparklesLWE, rendererClasses []}]
 userVariables [
  User.NewVector3f {type=Vector3f, description="", defaultValue {struct=Vector3f, value=<0 0 0>}}
  User.NormalizedNotifyProgress {type=NiagaraFloat, defaultValue=0, description=""}
  User.RibbonIDOffset {type=NiagaraInt32}
  User.Skeletal Mesh {
   type=NiagaraDataInterfaceSkeletalMesh
   defaultValue {
    struct=NiagaraExt_VariableValue_DataInterface
    value {dataInterface=&self:NiagaraDataInterfaceSkeletalMesh_0}
   }
  }
  User.SkeletalMesh {
   type=NiagaraDataInterfaceSkeletalMesh
   defaultValue {
    struct=NiagaraExt_VariableValue_DataInterface
    value {dataInterface=&self:NiagaraDataInterfaceSkeletalMesh_1}
   }
  }
 ]
}
