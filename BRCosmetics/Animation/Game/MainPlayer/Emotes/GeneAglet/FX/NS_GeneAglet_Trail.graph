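# NS_GeneAglet_Trail is a Niagara ribbon trail that follows a skeletal mesh for the GeneAglet player emote.
# Two emitters spawn short-lived anim trails between a pair of bone or socket indices on the user-supplied mesh, using the FN AnimTrail modules.
# SoftTrail draws a wider, near-black ribbon and SoftTrail001 a narrower, lighter purple-grey one, both fading out through color-curve alpha.
# The velocity-add modules are disabled.
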NiagaraSystem /BRCosmetics/Animation/Game/MainPlayer/Emotes/GeneAglet/FX/NS_GeneAglet_Trail
dependencies {usedDataInterfaces [&NiagaraDataInterfaceColorCurve], usedDynamicInputs [], usedModules [&SystemState]}
emitters {
 SoftTrail {
  emitterUpdateScript [
   &EmitterState()
   &FN_SpawnAnimTrail()
  ]
  particleSpawnScript [
   &FN_InitializeAnimTrailBoneIndecies(First Bone or Socket Index=1, Second Bone or Socket Index=0)
   &FN_InitializeAnimTrail(AnimTrailWidth=1.8, Color=<rgba 0 0 0 1>, Lifetime=0.1)
   &AddVelocity(Velocity=$User.NewVector3f, Velocity Speed Scale=100) {enabled=false}
  ]
  particleUpdateScript [
   &FN_UpdateAnimTrail(SkeletalMesh=$User.SkeletalMesh) {module=FN_UpdateAnimTrail001}
   &Color(Color=&ColorFromCurve(ColorCurve=DataInterface() {propertyValues {
       AlphaCurve {keys [<at 0 1 RCIM_Cubic>,<at 0.9811 0>]}
       BlueCurve {keys [<at 0 0.0132>]}
       GreenCurve {keys [<at 0 0.0061>]}
       RedCurve {keys [<at 0 0.026>]}
      }}))
   &SolveForcesAndVelocity(Clamp Velocity=false, Limit Acceleration=false)
   &DynamicMaterialParameters(Write Parameter Index 0=false)
  ]
  rendererClasses [&NiagaraRibbonRendererProperties]
  renderers [{rendererClass=&NiagaraRibbonRendererProperties}]
 }
 SoftTrail001 {
  emitterUpdateScript [
   &EmitterState()
   &FN_SpawnAnimTrail()
  ]
  particleSpawnScript [
   &FN_InitializeAnimTrailBoneIndecies(First Bone or Socket Index=1, Second Bone or Socket Index=0)
   &FN_InitializeAnimTrail(AnimTrailWidth=1.2, Color=<rgba 0 0 0 1>, Lifetime=0.1)
   &AddVelocity(Velocity=$User.NewVector3f, Velocity Speed Scale=100) {enabled=false}
  ]
  particleUpdateScript [
   &FN_UpdateAnimTrail(SkeletalMesh=$User.SkeletalMesh) {module=FN_UpdateAnimTrail001}
   &Color(Color=&ColorFromCurve(ColorCurve=DataInterface() {propertyValues {
       AlphaCurve {keys [<at 0 0.476 RCIM_Cubic>,<at 0.9811 0>]}
       BlueCurve {keys [<at 0 0.4184>]}
       GreenCurve {keys [<at 0 0.2876>]}
       RedCurve {keys [<at 0 0.4844>]}
      }}))
   &SolveForcesAndVelocity(Clamp Velocity=false, Limit Acceleration=false)
   &DynamicMaterialParameters(Write Parameter Index 0=false)
  ]
  rendererClasses [&NiagaraRibbonRendererProperties]
  renderers [{rendererClass=&NiagaraRibbonRendererProperties}]
 }
}
summary {
 systemName=NS_GeneAglet_Trail
 emitters [
  {emitterName=SoftTrail, rendererClasses [&NiagaraRibbonRendererProperties]}
  {emitterName=SoftTrail001, rendererClasses [&NiagaraRibbonRendererProperties]}
 ]
 userVariables [
  User.NewVector3f {type=Vector3f, description="", defaultValue {struct=Vector3f, value=<10 0 0>}}
  User.RibbonIDOffset {type=NiagaraInt32}
  User.SkeletalMesh {
   type=NiagaraDataInterfaceSkeletalMesh
   defaultValue {
    struct=NiagaraExt_VariableValue_DataInterface
    value {dataInterface=&self:NiagaraDataInterfaceSkeletalMesh_0}
   }
  }
 ]
}
